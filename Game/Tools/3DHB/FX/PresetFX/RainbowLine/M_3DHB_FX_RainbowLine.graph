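# Unlit, translucent, two-sided FX material for the 3DHB RainbowLine preset, enabled for Niagara sprites and mesh particles.
# A normal texture parameter drives the shared RainbowColor post function, whose BlendToEmissive output feeds the EmissiveAdjustments function.
# A separate branch derives an edge mask from the same normal by flattening it, dotting with +Z, applying a power and inverting.
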Material /Game/Tools/3DHB/FX/PresetFX/RainbowLine/M_3DHB_FX_RainbowLine
graph {
 properties {
  bUsedWithNiagaraMeshParticles=true
  bUsedWithNiagaraSprites=true
  blendMode=BLEND_Translucent
  pixelDepthOffsetMode=PDOM_Legacy
  shadingModel=MSM_Unlit
  twoSided=true
  usedShadingModels=Unlit
 }
 nodes [
  n1 = TextureSampleParameter2D() {parameterName=NormalTexture, samplerType=SAMPLERTYPE_Normal, texture=&FlatNormalBlue}
  MaterialFunctionCall(Emissive=MaterialFunctionCall!(
    Time
    "NormalIn(TS)"=@n1.RGB
    DirctionIn
    AdditionalDesaturationIn
    AdditionalStrengthIn
    AdditionalDensityIn
    AdditionalFlattenNormalIn
    WPOIn
   ) {materialFunction=/Game/Tools/3DHB/Shader/MasterLib/PostFunctions/MF_3DHB_POST_RainbowColor}.BlendToEmissive) {materialFunction=&MF_FX_EmissiveAdjustments}
  n2 = MaterialFunctionCall(Normal=@n1.RGB, Flatness=ScalarParameter() {defaultValue=0.7, parameterName=NormalToEdge_Flat}) {materialFunction=&FlattenNormal}
  n3 = Constant3Vector() {constant=<rgba 0 0 1 0>}
  OneMinus(None=Power(Base=DotProduct(A=@n2.Result, B=@n3), Exp=ScalarParameter() {defaultValue=20, parameterName=NormalToEdge_Power}))
 ]
}
